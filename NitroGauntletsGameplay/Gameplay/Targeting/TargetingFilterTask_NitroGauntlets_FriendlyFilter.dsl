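; Targeting filter task for Nitro Gauntlets that removes friendly actors from targeting results, except building actors.
; It breaks the hit result to get the hit actor and compares its team affiliation with the targeting request's source context.
; Friendly actors are removed unless they match an exception class array, which holds only BuildingActor.
; All non-friendly targets are kept.
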(fn ShouldRemoveTarget
  (params (TargetingHandle :type Targeting%20Request%20Handle%20Structure%20%28by%20ref%29) (TargetData :type Targeting%20Default%20Result%20Data%20Structure%20%28by%20ref%29))
  (bind _hitresult (Utilities|Struct|BreakTargetingDefaultResultData TargetData))
  (bind (_bblockinghit _binitialoverlap _time _distance _location _impactpoint _normal _impactnormal _physmat _hitactor) (Collision|BreakHitResult _hitresult))
  (if (Utilities|Enum|Equal(Enum) (Team|GetTeamAffiliation _hitactor (Utilities|Struct|BreakTargetingSourceContext (TargetingSystem|TargetingTypes|GetTargetingSourceContext (Targeting|GetTargetingSubsystem TargetingHandle) TargetingHandle))) "Friendly")
    (for _array_element (Utilities|Array|MakeArray "/Script/FortniteGame.BuildingActor")
      (if (Utilities|IsA _hitactor _array_element)
        (return false)))
    (return true)
    (else
      (return false))))
